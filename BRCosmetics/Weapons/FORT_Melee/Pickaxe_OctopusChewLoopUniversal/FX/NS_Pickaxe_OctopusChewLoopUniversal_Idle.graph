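# Idle Niagara effect for the OctopusChewLoopUniversal pickaxe, combining small drop sprites with tentacle ribbons placed through a socket reader.
# Drops spawn across six socket groups, with lifetime shrinking as owner speed rises, then drag, attraction and inherited velocity move them.
# Tentacles bursts a 20-particle ribbon between one of three randomly chosen socket pairs, using a randomized mid-arc offset, curl noise and staggered width animation along the ribbon.
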NiagaraSystem /BRCosmetics/Weapons/FORT_Melee/Pickaxe_OctopusChewLoopUniversal/FX/NS_Pickaxe_OctopusChewLoopUniversal_Idle
dependencies {
 usedDataInterfaces [&NiagaraDataInterfaceCurve, &NiagaraDataInterfaceSocketReader]
 usedDynamicInputs [&SimulationPosition, &UniformRangedFloat]
 usedModules [
  &self:SystemSpawnScript.SystemScriptSource.SystemScriptGraph.NiagaraNodeAssignment_0.SetVariables_#1
  &SystemState
 ]
}
emitters {
 Drops {
  emitterUpdateScript [
   &EmitterState()
   &SpawnRate(SpawnRate=4, Use Spawn Probability=false)
  ]
  particleSpawnScript [
   &FN_SetSpawnGroup(Number of Spawn Groups=6)
   &InitializeParticle(
    Color=<rgba 0.276 0.276 0.276 1>
    Lifetime=&FloatFromCurve(
     FloatCurve=DataInterface() {propertyValues {Curve {keys [<at 0 1>,<at 1 0>]}}}
     CurveIndex=&NormalizeFloat(Value=&VectorLength(Vector=$Engine.Owner.Velocity), Min=0, Max=650)
     Scale Curve=0.8
    )
    Sprite Size=&Vector2DFromFloat(Value=&RandomRangeFloat(Minimum=1, Maximum=3.5))
    Sprite Size Mode=::NewEnumerator4
    UsePositionOffset=false
    Write Material Random=true
   ) {module=InitializeParticle001}
   &SocketLocation(
    SamplePosOffset=false
    Socket Index=$Particles.SpawnGroup
    Socket Reader=$System.Socket Reader
    Socket Sampling Mode=::NewEnumerator3
    VelocitySpeedLimit=false
   )
   /Niagara/Modules/Update/SubUV/SubUVAnimation(SubUV Animation Mode=::NewEnumerator1) {module=SubUVAnimation001}
   &self:AddVelAlongSocketAxis(
    SocketAxis:Vector3f=&RandomRangeVector(Minimum=<1 -0.2 -0.2>, Maximum=<1 0.2 0.2>)
    Speed:NiagaraFloat=&RandomRangeFloat(Minimum=-25, Maximum=-90)
   ) {module=AddVelAlongSocketAxis}
  ]
  particleUpdateScript [
   &ParticleState()
   &InheritVelocity(
    Module.Limit Speed=true
    Velocity Scale=&VectorFromFloat(Value=&FloatFromCurve(
      FloatCurve=DataInterface() {propertyValues {Curve {keys [{arriveTangent=-1.9956, interpMode=RCIM_Cubic, leaveTangent=-1.9956, tangentMode=RCTM_User, time=0, value=1}, <at 1 0 RCIM_Cubic>]}}}
      Scale Curve=0.1
     ))
   ) {module=InheritVelocity001}
   &Drag(Drag=3, Use Linear Drag=true, Use Rotational Drag=false)
   &PointAttractionForce(
    Attraction Radius=300
    Attraction Strength=10
    Attractor Position:Vector3f=$Particles.SkeletalMeshLocation.SampledPosition
    Falloff Exponent=0.7
    Kill Within Radius=false
    Module.Use Falloff=true
    UseAttractorOffset=false
   )
   &SolveForcesAndVelocity(Clamp Velocity=false, Limit Acceleration=false)
   &ScaleSpriteSize(Scale Factor=&Vector2DFromFloat(Value=&FloatFromCurve(FloatCurve=DataInterface() {propertyValues {Curve {keys [<at 0 1 RCIM_Cubic>, {arriveTangent=-1.5322, interpMode=RCIM_Cubic, leaveTangent=-1.5322, tangentMode=RCTM_User, time=1, value=0}]}}})))
  ]
  rendererClasses [&NiagaraSpriteRendererProperties]
  renderers [{rendererClass=&NiagaraSpriteRendererProperties}]
 }
 Tentacles {
  emitterUpdateScript [
   &EmitterState()
   /SetVariables(Emitter.RandomRibbon:NiagaraInt32=&RandomRangeInt(Minimum=0, Maximum=5000)) {isSetParameters=true}
   &SpawnBurst_Instantaneous(Spawn Count=20, Use Spawn Probability=true)
  ]
  particleSpawnScript [
   &PartitionParticles(
    Number Of Partitions=1
    Particles Per Partition=20
    Partition Mode:ENiagara_PartitionMode=::NewEnumerator1
    Write Partition Index to Ribbon Link Order:ENiagara_AttributeSamplingApplyOutput=::NewEnumerator0
   )
   /SetVariables(
    Particles.RandomIndexPerRibbon:NiagaraInt32=&ModuloInt(A=$Emitter.RandomRibbon, B=3)
    Particles.RibbonID=/Niagara/DynamicInputs/NiagaraID/MakeNiagaraIDFromInt(Index:NiagaraInt32=$Emitter.RandomRibbon, Acquire Tag:NiagaraInt32=0)
   ) {isSetParameters=true}
   &InitializeParticle(
    Color=<rgba 1 0.9084 0.7325 1>
    Lifetime=&Multiply_Float(
     A=&RandomRangeFloat(Minimum=0.75, Maximum=1)
     B=&RandomRangeFloat(Minimum=1, Maximum=1.3, Randomness Mode=::NewEnumerator1, Random Seed=$Emitter.RandomRibbon, Fixed Random Seed=true)
    )
    Ribbon Width=&Multiply_Float(
     A=&RandomRangeFloat(Minimum=2, Maximum=4)
     B=&MergeHelper_FloatSwitch(0=1, 1=0.8, 2=0.8, 3=?, Index=$Particles.RandomIndexPerRibbon, Num=::NewEnumerator1)
    )
    Ribbon Width Mode=::NewEnumerator1
    UsePositionOffset=false
    Write Material Random=true
   )
   &SocketLocation(
    Position Sampling=::NewEnumerator1
    SamplePosOffset=false
    Socket Index=&Multiply_Int(A=$Particles.RandomIndexPerRibbon, B=2)
    Socket Reader=$System.Socket Reader
    Socket Sampling Mode=::NewEnumerator3
    VelocitySpeedLimit=false
   )
   &SocketLocation(
    Position Sampling=::NewEnumerator1
    SamplePosOffset=false
    Socket Index=&Add_Integer(A=&Multiply_Int(A=$Particles.RandomIndexPerRibbon, B=2), B=1)
    Socket Reader=$System.Socket Reader
    Socket Sampling Mode=::NewEnumerator3
    VelocitySpeedLimit=false
   ) {module=SocketLocation001}
   /SetVariables(Particles.Position=&AddVectorToPosition(
     Position=&Multiply_VectorByFloat(
      Vector=&RandomRangeVector(Minimum=<-1 -1 -1>, Maximum=<1 1 1>)
      Float=&FloatFromCurve(
       FloatCurve=DataInterface() {propertyValues {Curve {keys [
           {arriveTangent=0.1446, interpMode=RCIM_Cubic, leaveTangent=0.1446, tangentMode=RCTM_User, time=0, value=0}
           <at 0.5 1 RCIM_Cubic>
           {arriveTangent=0.3527, interpMode=RCIM_Cubic, leaveTangent=0.3527, tangentMode=RCTM_User, time=1, value=0}
          ]}}}
       CurveIndex=$Particles.RibbonLinkOrder
       Scale Curve=&MergeHelper_FloatSwitch(0=9, 1=5, 2=5, 3=?, Index=$Particles.RandomIndexPerRibbon, Num=::NewEnumerator1)
      )
     )
     Vector=&Lerp_Vector(A=$Output.SocketLocation.SampledPosition, B=$Output.SocketLocation001.SampledPosition, Alpha=$Particles.RibbonLinkOrder)
    )) {isSetParameters=true}
   &AddVelocity(
    Velocity=&RandomRangeVector(Minimum=<-1 -1 -1>, Maximum=<1 1 1>)
    Velocity Speed Scale=&Multiply_Float(
     A=&RandomRangeFloat(Minimum=5, Maximum=10)
     B=&FloatFromCurve(
      FloatCurve=DataInterface() {propertyValues {Curve {keys [
          {arriveTangent=3.717, interpMode=RCIM_Cubic, leaveTangent=3.717, tangentMode=RCTM_User, time=0, value=0}
          <at 0.5 1 RCIM_Cubic>
          {arriveTangent=-3.1345, interpMode=RCIM_Cubic, leaveTangent=-3.1345, tangentMode=RCTM_User, time=1, value=0}
         ]}}}
      CurveIndex=$Particles.RibbonLinkOrder
     )
    )
   ) {enabled=false}
  ]
  particleUpdateScript [
   &CurlNoiseForce(
    Bias Noise=false
    Noise Frequency=&RandomRangeFloat(Minimum=200, Maximum=300)
    Noise Strength=&RandomRangeFloat(Minimum=200, Maximum=250)
    Pan Noise=false
   )
   &Drag(Drag=&RandomRangeFloat(Minimum=20, Maximum=30), Use Linear Drag=true, Use Rotational Drag=true)
   &SolveForcesAndVelocity(Clamp Velocity=false, Limit Acceleration=false)
   &ParticleState(Kill Particles When Lifetime Has Elapsed=false)
   &ScaleRibbonWidth(Ribbon Width Scale=&Multiply_Float(
     A=&Multiply_Float(
      A=&Multiply_Float(
       A=&FloatFromCurve(FloatCurve=DataInterface() {propertyValues {Curve {keys [
            {arriveTangent=10.48, interpMode=RCIM_Cubic, leaveTangent=10.48, tangentMode=RCTM_User, time=0, value=0}
            {arriveTangent=2, leaveTangent=2, time=0.2, value=1}
            {arriveTangent=-1.25, interpMode=RCIM_Cubic, leaveTangent=-1.25, time=0.5, value=1}
            {interpMode=RCIM_Cubic, leaveTangent=-1.6768, time=1, value=0}
           ]}}})
       B=&FloatFromCurve(
        FloatCurve=DataInterface() {propertyValues {Curve {keys [
            {arriveTangent=5.8199, interpMode=RCIM_Cubic, time=0, value=0}
            {arriveTangent=2, leaveTangent=2, time=0.15, value=1}
            <at 0.5 1>
            {arriveTangent=-2, interpMode=RCIM_Cubic, leaveTangent=-2, time=0.85, value=1}
            {interpMode=RCIM_Cubic, leaveTangent=-3.1345, time=1, value=0}
           ]}}}
        CurveIndex=$Particles.RibbonLinkOrder
       )
      )
      B=&Waveform(
       Waveform Count=::NewEnumerator1
       Global Input Position=&RandomRangeFloat(Minimum=0, Maximum=100)
       "Waveform [1]"=::NewEnumerator0
       "[1] Frequency"=0.3
       "[1] Amplitude Min/Max"=<0.6 1>
       "Waveform [2] Blend"=::NewEnumerator2
       "[2] Input Position"=0
       "[2] Phase"=$Emitter.Age
       "[2] Phase Scale"=1
       "[2] Compound Frequency"=2
       "[2] Frequency"=3.7
       "[2] Amplitude Min/Max"=<1 1>
       "[2] Scale"=1
       "[2] Bias"=0
       "[2] Exponent"=1
       "[2] Pulse Duration"=0.1
       "[2] Clip Triangle Min/Max"=<0 0>
       "[2] Random Seed"=0
      )
     )
     B=&FloatFromCurve(
      FloatCurve=DataInterface() {propertyValues {Curve {keys [<at 0 0 RCIM_Cubic>,<at 0.1 1 RCIM_Cubic>]}}}
      CurveIndex=&Subtract_Float(A=$Particles.Age, B=&Multiply_Float(A=$Particles.RibbonLinkOrder, B=0.5))
     )
    ))
   /SetVariables(Particles.Position=&AddVectorToPosition(
     Position=$Particles.Initial.Position
     Vector=&Multiply_VectorByFloat(
      Vector=&RandomRangeVector(Minimum=<-1 -1 -1>, Maximum=<1 1 1>)
      Float=&Multiply_Float(
       A=&FloatFromCurve(
        FloatCurve=DataInterface() {propertyValues {Curve {keys [
            {arriveTangent=5.8199, interpMode=RCIM_Cubic, time=0, value=0}
            {arriveTangent=2, leaveTangent=2, time=0.2, value=1}
            <at 0.5 1>
            {arriveTangent=-2, interpMode=RCIM_Cubic, leaveTangent=-2, time=0.8, value=1}
            {interpMode=RCIM_Cubic, leaveTangent=-3.1345, time=1, value=0}
           ]}}}
        CurveIndex=$Particles.RibbonLinkOrder
        Scale Curve=&RandomRangeFloat(Minimum=5, Maximum=10)
       )
       B=&FloatFromCurve(FloatCurve=DataInterface() {propertyValues {Curve {keys [
            {arriveTangent=1.2974, interpMode=RCIM_Cubic, leaveTangent=1.2974, tangentMode=RCTM_User, time=0, value=0.5}
            {interpMode=RCIM_Cubic, leaveTangent=-0.2, time=1, value=1}
           ]}}})
      )
     )
    )) {enabled=false, isSetParameters=true}
   &DynamicMaterialParameters(
    Index 0 Param 1=&FloatFromCurve(FloatCurve=DataInterface() {propertyValues {Curve {keys [<at 0 0 RCIM_Cubic>,<at 0.5 0 RCIM_Cubic>,<at 1 0 RCIM_Cubic>]}}})
    Module.Param0WriteEnabled=false
    Module.Param1WriteEnabled=false
    Module.Param2WriteEnabled=false
    Module.Param3WriteEnabled=false
   )
   &KillParticles(Kill Particles=&SetBoolByFloatComparison(A=$Particles.Age, B=1.25))
  ]
  rendererClasses [&NiagaraRibbonRendererProperties]
  renderers [{rendererClass=&NiagaraRibbonRendererProperties}]
 }
}
summary {
 systemName=NS_Pickaxe_OctopusChewLoopUniversal_Idle
 emitters [
  {emitterName=Drops, rendererClasses [&NiagaraSpriteRendererProperties]}
  {emitterName=Tentacles, rendererClasses [&NiagaraRibbonRendererProperties]}
 ]
 userVariables []
}
