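# Control Rig for the OctopusChewLoopUniversal two-handed pickaxe that procedurally wiggles its tend* bone chains, apparently tentacles.
# Begin Execution runs 16 ChainHarmonics units in sequence, each rooted on one chain or sub-chain under the handle bone.
# Only noisy Wave motion is enabled, with pendulum and reach off, and frequency, amplitude and easing vary between chains.
# The tendA and tendJ chains get extra units on child bones, and every unit has debug drawing on.
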ControlRigBlueprint /BRCosmetics/Weapons/FORT_Melee/Pickaxe_OctopusChewLoopUniversal/Meshes/OctopusChewLoopUniversal_Axe_CtrlRig
graphs [
 RigVMFunctionLibrary {links [], nodes []}
 RigVMModel {nodes [
   RigUnit_BeginExecution = &RigUnit_BeginExecution()
   /Script/ControlRig.RigUnit_ChainHarmonics(
    ChainRoot=tendA_2
    DrawWorldOffset=(Rotation=(X=0,Y=0,Z=0,W=1),Translation=(X=0,Y=0,Z=0),Scale3D=(X=1,Y=1,Z=1))
    Pendulum=(bEnabled=False,PendulumStiffness=2,PendulumGravity=(X=0,Y=0,Z=0),PendulumBlend=0.75,PendulumDrag=0.98,PendulumMinimum=0,PendulumMaximum=0.1,PendulumEase=Linear,UnwindAxis=(X=0,Y=1,Z=0),UnwindMinimum=0.2,UnwindMaximum=0.05)
    Reach=(bEnabled=False,ReachTarget=(X=0,Y=0,Z=0),ReachAxis=(X=1,Y=0,Z=0),ReachMinimum=0,ReachMaximum=0,ReachEase=SineEaseInOut)
    Speed=(X=0.25,Y=0.25,Z=0.1)
    Wave=(bEnabled=True,WaveFrequency=(X=0.5,Y=0.5,Z=0.5),WaveAmplitude=(X=0.25,Y=2.5,Z=1.5),WaveOffset=(X=1,Y=0,Z=2),WaveNoise=(X=8,Y=8,Z=30),WaveMinimum=0,WaveMaximum=0.75,WaveEase=SineEaseIn)
    WaveCurve=(EditorCurveData=(Keys=((Time=-0.0198,Value=0.4929),(Time=0.9855,Value=0.5049)),PreInfinityExtrap=RCCE_Constant,PostInfinityExtrap=RCCE_Constant,DefaultValue=340282346638528859811704183484516925440))
    bDrawDebug=True
    ExecutePin=@RigUnit_BeginExecution.ExecutePin
   ) {is=struct RigVMUnitNode}
   > &RigUnit_ChainHarmonics(
    ChainRoot=tendB_1
    DrawWorldOffset=(Rotation=(X=0,Y=0,Z=0,W=1),Translation=(X=0,Y=0,Z=0),Scale3D=(X=1,Y=1,Z=1))
    Pendulum=(bEnabled=False,PendulumStiffness=2,PendulumGravity=(X=0,Y=0,Z=0),PendulumBlend=0.75,PendulumDrag=0.98,PendulumMinimum=0,PendulumMaximum=0.1,PendulumEase=Linear,UnwindAxis=(X=0,Y=1,Z=0),UnwindMinimum=0.2,UnwindMaximum=0.05)
    Reach=(bEnabled=False,ReachTarget=(X=0,Y=0,Z=0),ReachAxis=(X=1,Y=0,Z=0),ReachMinimum=0,ReachMaximum=0,ReachEase=SineEaseInOut)
    Speed=(X=0.25,Y=0.25,Z=0.1)
    Wave=(bEnabled=True,WaveFrequency=(X=0.5,Y=0.5,Z=0.5),WaveAmplitude=(X=0.25,Y=2.5,Z=1.5),WaveOffset=(X=1,Y=0,Z=2),WaveNoise=(X=8,Y=8,Z=30),WaveMinimum=0,WaveMaximum=0.75,WaveEase=SineEaseIn)
    WaveCurve=(EditorCurveData=(Keys=((Time=-0.0198,Value=0.4929),(Time=0.9855,Value=0.5049)),PreInfinityExtrap=RCCE_Constant,PostInfinityExtrap=RCCE_Constant,DefaultValue=340282346638528859811704183484516925440))
    bDrawDebug=True
   )
   > &RigUnit_ChainHarmonics(
    ChainRoot=tendC_1
    DrawWorldOffset=(Rotation=(X=0,Y=0,Z=0,W=1),Translation=(X=0,Y=0,Z=0),Scale3D=(X=1,Y=1,Z=1))
    Pendulum=(bEnabled=False,PendulumStiffness=2,PendulumGravity=(X=0,Y=0,Z=0),PendulumBlend=0.75,PendulumDrag=0.98,PendulumMinimum=0,PendulumMaximum=0.1,PendulumEase=Linear,UnwindAxis=(X=0,Y=1,Z=0),UnwindMinimum=0.2,UnwindMaximum=0.05)
    Reach=(bEnabled=False,ReachTarget=(X=0,Y=0,Z=0),ReachAxis=(X=1,Y=0,Z=0),ReachMinimum=0,ReachMaximum=0,ReachEase=SineEaseInOut)
    Speed=(X=0.25,Y=0.25,Z=0.1)
    Wave=(bEnabled=True,WaveFrequency=(X=0.5,Y=0.5,Z=0.5),WaveAmplitude=(X=0.25,Y=2.5,Z=1.5),WaveOffset=(X=1,Y=0,Z=2),WaveNoise=(X=8,Y=8,Z=30),WaveMinimum=0,WaveMaximum=0.75,WaveEase=SineEaseIn)
    WaveCurve=(EditorCurveData=(Keys=((Time=-0.0198,Value=0.4929),(Time=0.9855,Value=0.5049)),PreInfinityExtrap=RCCE_Constant,PostInfinityExtrap=RCCE_Constant,DefaultValue=340282346638528859811704183484516925440))
    bDrawDebug=True
   )
   > &RigUnit_ChainHarmonics(
    ChainRoot=tendD_1
    DrawWorldOffset=(Rotation=(X=0,Y=0,Z=0,W=1),Translation=(X=0,Y=0,Z=0),Scale3D=(X=1,Y=1,Z=1))
    Pendulum=(bEnabled=False,PendulumStiffness=2,PendulumGravity=(X=0,Y=0,Z=0),PendulumBlend=0.75,PendulumDrag=0.98,PendulumMinimum=0,PendulumMaximum=0.1,PendulumEase=Linear,UnwindAxis=(X=0,Y=1,Z=0),UnwindMinimum=0.2,UnwindMaximum=0.05)
    Reach=(bEnabled=False,ReachTarget=(X=0,Y=0,Z=0),ReachAxis=(X=1,Y=0,Z=0),ReachMinimum=0,ReachMaximum=0,ReachEase=SineEaseInOut)
    Speed=(X=0.25,Y=0.25,Z=0.1)
    Wave=(bEnabled=True,WaveFrequency=(X=0.5,Y=0.5,Z=0.5),WaveAmplitude=(X=0.25,Y=2.5,Z=1.5),WaveOffset=(X=1,Y=0,Z=2),WaveNoise=(X=8,Y=8,Z=30),WaveMinimum=0,WaveMaximum=0.75,WaveEase=SineEaseIn)
    WaveCurve=(EditorCurveData=(Keys=((Time=-0.0198,Value=0.4929),(Time=0.9855,Value=0.5049)),PreInfinityExtrap=RCCE_Constant,PostInfinityExtrap=RCCE_Constant,DefaultValue=340282346638528859811704183484516925440))
    bDrawDebug=True
   )
   > &RigUnit_ChainHarmonics(
    ChainRoot=tendE
    DrawWorldOffset=(Rotation=(X=0,Y=0,Z=0,W=1),Translation=(X=0,Y=0,Z=0),Scale3D=(X=1,Y=1,Z=1))
    Pendulum=(bEnabled=False,PendulumStiffness=2,PendulumGravity=(X=0,Y=0,Z=0),PendulumBlend=0.75,PendulumDrag=0.98,PendulumMinimum=0,PendulumMaximum=0.1,PendulumEase=Linear,UnwindAxis=(X=0,Y=1,Z=0),UnwindMinimum=0.2,UnwindMaximum=0.05)
    Reach=(bEnabled=False,ReachTarget=(X=0,Y=0,Z=0),ReachAxis=(X=1,Y=0,Z=0),ReachMinimum=0,ReachMaximum=0,ReachEase=SineEaseInOut)
    Speed=(X=0.25,Y=0.25,Z=0.1)
    Wave=(bEnabled=True,WaveFrequency=(X=0.5,Y=0.5,Z=0.5),WaveAmplitude=(X=0.25,Y=2.5,Z=1.5),WaveOffset=(X=1,Y=0,Z=2),WaveNoise=(X=8,Y=8,Z=30),WaveMinimum=0,WaveMaximum=0.75,WaveEase=SineEaseIn)
    WaveCurve=(EditorCurveData=(Keys=((Time=-0.0198,Value=0.4929),(Time=0.9855,Value=0.5049)),PreInfinityExtrap=RCCE_Constant,PostInfinityExtrap=RCCE_Constant,DefaultValue=340282346638528859811704183484516925440))
    bDrawDebug=True
   )
   > &RigUnit_ChainHarmonics(
    ChainRoot=tendF_1
    DrawWorldOffset=(Rotation=(X=0,Y=0,Z=0,W=1),Translation=(X=0,Y=0,Z=0),Scale3D=(X=1,Y=1,Z=1))
    Pendulum=(bEnabled=False,PendulumStiffness=2,PendulumGravity=(X=0,Y=0,Z=0),PendulumBlend=0.75,PendulumDrag=0.98,PendulumMinimum=0,PendulumMaximum=0.1,PendulumEase=Linear,UnwindAxis=(X=0,Y=1,Z=0),UnwindMinimum=0.2,UnwindMaximum=0.05)
    Reach=(bEnabled=False,ReachTarget=(X=0,Y=0,Z=0),ReachAxis=(X=1,Y=0,Z=0),ReachMinimum=0,ReachMaximum=0,ReachEase=SineEaseInOut)
    Speed=(X=0.25,Y=0.25,Z=0.1)
    Wave=(bEnabled=True,WaveFrequency=(X=0.5,Y=0.5,Z=0.5),WaveAmplitude=(X=0.25,Y=2.5,Z=1.5),WaveOffset=(X=1,Y=0,Z=2),WaveNoise=(X=8,Y=8,Z=30),WaveMinimum=0,WaveMaximum=0.75,WaveEase=SineEaseIn)
    WaveCurve=(EditorCurveData=(Keys=((Time=-0.0198,Value=0.4929),(Time=0.9855,Value=0.5049)),PreInfinityExtrap=RCCE_Constant,PostInfinityExtrap=RCCE_Constant,DefaultValue=340282346638528859811704183484516925440))
    bDrawDebug=True
   )
   > &RigUnit_ChainHarmonics(
    ChainRoot=tendA_1
    DrawWorldOffset=(Rotation=(X=0,Y=0,Z=0,W=1),Translation=(X=0,Y=0,Z=0),Scale3D=(X=1,Y=1,Z=1))
    Pendulum=(bEnabled=False,PendulumStiffness=2,PendulumGravity=(X=0,Y=0,Z=0),PendulumBlend=0.75,PendulumDrag=0.98,PendulumMinimum=0,PendulumMaximum=0.1,PendulumEase=Linear,UnwindAxis=(X=0,Y=1,Z=0),UnwindMinimum=0.2,UnwindMaximum=0.05)
    Reach=(bEnabled=False,ReachTarget=(X=0,Y=0,Z=0),ReachAxis=(X=1,Y=0,Z=0),ReachMinimum=0,ReachMaximum=0,ReachEase=SineEaseInOut)
    Speed=(X=0.25,Y=0.25,Z=0.1)
    Wave=(bEnabled=True,WaveFrequency=(X=0.2,Y=0.2,Z=0.2),WaveAmplitude=(X=0.25,Y=2.5,Z=1.5),WaveOffset=(X=1,Y=0,Z=2),WaveNoise=(X=8,Y=8,Z=30),WaveMinimum=0,WaveMaximum=0.75,WaveEase=SineEaseIn)
    WaveCurve=(EditorCurveData=(Keys=((Time=-0.0198,Value=0.4929),(Time=0.9855,Value=0.5049)),PreInfinityExtrap=RCCE_Constant,PostInfinityExtrap=RCCE_Constant,DefaultValue=340282346638528859811704183484516925440))
    bDrawDebug=True
   )
   > &RigUnit_ChainHarmonics(
    ChainRoot=tendG
    DrawWorldOffset=(Rotation=(X=0,Y=0,Z=0,W=1),Translation=(X=0,Y=0,Z=0),Scale3D=(X=1,Y=1,Z=1))
    Pendulum=(bEnabled=False,PendulumStiffness=2,PendulumGravity=(X=0,Y=0,Z=0),PendulumBlend=0.75,PendulumDrag=0.98,PendulumMinimum=0,PendulumMaximum=0.1,PendulumEase=Linear,UnwindAxis=(X=0,Y=1,Z=0),UnwindMinimum=0.2,UnwindMaximum=0.05)
    Reach=(bEnabled=False,ReachTarget=(X=0,Y=0,Z=0),ReachAxis=(X=1,Y=0,Z=0),ReachMinimum=0,ReachMaximum=0,ReachEase=SineEaseInOut)
    Speed=(X=0.25,Y=0.25,Z=0.1)
    Wave=(bEnabled=True,WaveFrequency=(X=0.2,Y=0.2,Z=0.2),WaveAmplitude=(X=0.25,Y=2.5,Z=1.5),WaveOffset=(X=1,Y=0,Z=2),WaveNoise=(X=8,Y=8,Z=30),WaveMinimum=0,WaveMaximum=0.75,WaveEase=SineEaseIn)
    WaveCurve=(EditorCurveData=(Keys=((Time=-0.0198,Value=0.4929),(Time=0.9855,Value=0.5049)),PreInfinityExtrap=RCCE_Constant,PostInfinityExtrap=RCCE_Constant,DefaultValue=340282346638528859811704183484516925440))
    bDrawDebug=True
   )
   > &RigUnit_ChainHarmonics(
    ChainRoot=tendI_1
    DrawWorldOffset=(Rotation=(X=0,Y=0,Z=0,W=1),Translation=(X=0,Y=0,Z=0),Scale3D=(X=1,Y=1,Z=1))
    Pendulum=(bEnabled=False,PendulumStiffness=2,PendulumGravity=(X=0,Y=0,Z=0),PendulumBlend=0.75,PendulumDrag=0.98,PendulumMinimum=0,PendulumMaximum=0.1,PendulumEase=Linear,UnwindAxis=(X=0,Y=1,Z=0),UnwindMinimum=0.2,UnwindMaximum=0.05)
    Reach=(bEnabled=False,ReachTarget=(X=0,Y=0,Z=0),ReachAxis=(X=1,Y=0,Z=0),ReachMinimum=0,ReachMaximum=0,ReachEase=SineEaseInOut)
    Speed=(X=0.25,Y=0.25,Z=0.1)
    Wave=(bEnabled=True,WaveFrequency=(X=0.2,Y=0.2,Z=0.2),WaveAmplitude=(X=0.5,Y=1.5,Z=1.5),WaveOffset=(X=1,Y=0,Z=2),WaveNoise=(X=8,Y=8,Z=30),WaveMinimum=0,WaveMaximum=0.75,WaveEase=SineEaseIn)
    WaveCurve=(EditorCurveData=(Keys=((Time=-0.0198,Value=0.4929),(Time=0.9855,Value=0.5049)),PreInfinityExtrap=RCCE_Constant,PostInfinityExtrap=RCCE_Constant,DefaultValue=340282346638528859811704183484516925440))
    bDrawDebug=True
   )
   > &RigUnit_ChainHarmonics(
    ChainRoot=tendH_1
    DrawWorldOffset=(Rotation=(X=0,Y=0,Z=0,W=1),Translation=(X=0,Y=0,Z=0),Scale3D=(X=1,Y=1,Z=1))
    Pendulum=(bEnabled=False,PendulumStiffness=2,PendulumGravity=(X=0,Y=0,Z=0),PendulumBlend=0.75,PendulumDrag=0.98,PendulumMinimum=0,PendulumMaximum=0.1,PendulumEase=Linear,UnwindAxis=(X=0,Y=1,Z=0),UnwindMinimum=0.2,UnwindMaximum=0.05)
    Reach=(bEnabled=False,ReachTarget=(X=0,Y=0,Z=0),ReachAxis=(X=1,Y=0,Z=0),ReachMinimum=0,ReachMaximum=0,ReachEase=SineEaseInOut)
    Speed=(X=0.25,Y=0.25,Z=0.1)
    Wave=(bEnabled=True,WaveFrequency=(X=0.2,Y=0.2,Z=0.2),WaveAmplitude=(X=0.25,Y=1,Z=0.5),WaveOffset=(X=1,Y=0,Z=2),WaveNoise=(X=8,Y=8,Z=30),WaveMinimum=0,WaveMaximum=0.75,WaveEase=SineEaseIn)
    WaveCurve=(EditorCurveData=(Keys=((Time=-0.0198,Value=0.4929),(Time=0.9855,Value=0.5049)),PreInfinityExtrap=RCCE_Constant,PostInfinityExtrap=RCCE_Constant,DefaultValue=340282346638528859811704183484516925440))
    bDrawDebug=True
   )
   > &RigUnit_ChainHarmonics(
    ChainRoot=tendJ_1
    DrawWorldOffset=(Rotation=(X=0,Y=0,Z=0,W=1),Translation=(X=0,Y=0,Z=0),Scale3D=(X=1,Y=1,Z=1))
    Pendulum=(bEnabled=False,PendulumStiffness=2,PendulumGravity=(X=0,Y=0,Z=0),PendulumBlend=0.75,PendulumDrag=0.98,PendulumMinimum=0,PendulumMaximum=0.1,PendulumEase=Linear,UnwindAxis=(X=0,Y=1,Z=0),UnwindMinimum=0.2,UnwindMaximum=0.05)
    Reach=(bEnabled=False,ReachTarget=(X=0,Y=0,Z=0),ReachAxis=(X=1,Y=0,Z=0),ReachMinimum=0,ReachMaximum=0,ReachEase=SineEaseInOut)
    Speed=(X=0.25,Y=0.25,Z=0.1)
    Wave=(bEnabled=True,WaveFrequency=(X=2,Y=2,Z=1),WaveAmplitude=(X=1,Y=1.5,Z=0.5),WaveOffset=(X=1,Y=0,Z=1),WaveNoise=(X=8,Y=8,Z=30),WaveMinimum=0,WaveMaximum=0.75,WaveEase=BounceEaseIn)
    WaveCurve=(EditorCurveData=(Keys=((Time=-0.0198,Value=0.4929),(Time=0.9855,Value=0.5049)),PreInfinityExtrap=RCCE_Constant,PostInfinityExtrap=RCCE_Constant,DefaultValue=340282346638528859811704183484516925440))
    bDrawDebug=true
   )
   > &RigUnit_ChainHarmonics(
    ChainRoot=tendJ_2b
    DrawWorldOffset=(Rotation=(X=0,Y=0,Z=0,W=1),Translation=(X=0,Y=0,Z=0),Scale3D=(X=1,Y=1,Z=1))
    Pendulum=(bEnabled=False,PendulumStiffness=2,PendulumGravity=(X=0,Y=0,Z=0),PendulumBlend=0.75,PendulumDrag=0.98,PendulumMinimum=0,PendulumMaximum=0.1,PendulumEase=Linear,UnwindAxis=(X=0,Y=1,Z=0),UnwindMinimum=0.2,UnwindMaximum=0.05)
    Reach=(bEnabled=False,ReachTarget=(X=0,Y=0,Z=0),ReachAxis=(X=1,Y=0,Z=0),ReachMinimum=0,ReachMaximum=0,ReachEase=SineEaseInOut)
    Speed=(X=0.25,Y=0.25,Z=0.1)
    Wave=(bEnabled=True,WaveFrequency=(X=2,Y=2,Z=1),WaveAmplitude=(X=0.2,Y=0.2,Z=0.3),WaveOffset=(X=2,Y=3,Z=1),WaveNoise=(X=8,Y=8,Z=30),WaveMinimum=0,WaveMaximum=0.75,WaveEase=BounceEaseIn)
    WaveCurve=(EditorCurveData=(Keys=((Time=-0.0198,Value=0.4929),(Time=0.9855,Value=0.5049)),PreInfinityExtrap=RCCE_Constant,PostInfinityExtrap=RCCE_Constant,DefaultValue=340282346638528859811704183484516925440))
    bDrawDebug=true
   )
   > &RigUnit_ChainHarmonics(
    ChainRoot=tendK_1
    DrawWorldOffset=(Rotation=(X=0,Y=0,Z=0,W=1),Translation=(X=0,Y=0,Z=0),Scale3D=(X=1,Y=1,Z=1))
    Pendulum=(bEnabled=False,PendulumStiffness=2,PendulumGravity=(X=0,Y=0,Z=0),PendulumBlend=0.75,PendulumDrag=0.98,PendulumMinimum=0,PendulumMaximum=0.1,PendulumEase=Linear,UnwindAxis=(X=0,Y=1,Z=0),UnwindMinimum=0.2,UnwindMaximum=0.05)
    Reach=(bEnabled=False,ReachTarget=(X=0,Y=0,Z=0),ReachAxis=(X=1,Y=0,Z=0),ReachMinimum=0,ReachMaximum=0,ReachEase=SineEaseInOut)
    Speed=(X=0.25,Y=0.25,Z=0.1)
    Wave=(bEnabled=True,WaveFrequency=(X=1,Y=1,Z=0.5),WaveAmplitude=(X=0.25,Y=2.5,Z=1.5),WaveOffset=(X=1,Y=0,Z=2),WaveNoise=(X=8,Y=8,Z=30),WaveMinimum=0,WaveMaximum=0.75,WaveEase=Linear)
    WaveCurve=(EditorCurveData=(Keys=((Time=-0.0198,Value=0.4929),(Time=0.9855,Value=0.5049)),PreInfinityExtrap=RCCE_Constant,PostInfinityExtrap=RCCE_Constant,DefaultValue=340282346638528859811704183484516925440))
    bDrawDebug=True
   )
   > &RigUnit_ChainHarmonics(
    ChainRoot=tendL_1
    DrawWorldOffset=(Rotation=(X=0,Y=0,Z=0,W=1),Translation=(X=0,Y=0,Z=0),Scale3D=(X=1,Y=1,Z=1))
    Pendulum=(bEnabled=False,PendulumStiffness=2,PendulumGravity=(X=0,Y=0,Z=0),PendulumBlend=0.75,PendulumDrag=0.98,PendulumMinimum=0,PendulumMaximum=0.1,PendulumEase=Linear,UnwindAxis=(X=0,Y=1,Z=0),UnwindMinimum=0.2,UnwindMaximum=0.05)
    Reach=(bEnabled=False,ReachTarget=(X=0,Y=0,Z=0),ReachAxis=(X=1,Y=0,Z=0),ReachMinimum=0,ReachMaximum=0,ReachEase=SineEaseInOut)
    Speed=(X=0.25,Y=0.25,Z=0.1)
    Wave=(bEnabled=True,WaveFrequency=(X=3,Y=2,Z=1),WaveAmplitude=(X=3,Y=3,Z=1),WaveOffset=(X=1,Y=0,Z=2),WaveNoise=(X=8,Y=8,Z=30),WaveMinimum=0,WaveMaximum=0.75,WaveEase=SineEaseIn)
    WaveCurve=(EditorCurveData=(Keys=((Time=-0.0198,Value=0.4929),(Time=0.9855,Value=0.5049)),PreInfinityExtrap=RCCE_Constant,PostInfinityExtrap=RCCE_Constant,DefaultValue=340282346638528859811704183484516925440))
    bDrawDebug=True
   )
   > &RigUnit_ChainHarmonics(
    ChainRoot=tendM_1
    DrawWorldOffset=(Rotation=(X=0,Y=0,Z=0,W=1),Translation=(X=0,Y=0,Z=0),Scale3D=(X=1,Y=1,Z=1))
    Pendulum=(bEnabled=False,PendulumStiffness=2,PendulumGravity=(X=0,Y=0,Z=0),PendulumBlend=0.75,PendulumDrag=0.98,PendulumMinimum=0,PendulumMaximum=0.1,PendulumEase=Linear,UnwindAxis=(X=0,Y=1,Z=0),UnwindMinimum=0.2,UnwindMaximum=0.05)
    Reach=(bEnabled=False,ReachTarget=(X=0,Y=0,Z=0),ReachAxis=(X=1,Y=0,Z=0),ReachMinimum=0,ReachMaximum=0,ReachEase=SineEaseInOut)
    Speed=(X=0.25,Y=0.25,Z=0.1)
    Wave=(bEnabled=True,WaveFrequency=(X=1,Y=1,Z=0.2),WaveAmplitude=(X=3,Y=0.5,Z=0.5),WaveOffset=(X=1,Y=0,Z=2),WaveNoise=(X=8,Y=8,Z=30),WaveMinimum=0,WaveMaximum=0.75,WaveEase=Linear)
    WaveCurve=(EditorCurveData=(Keys=((Time=-0.0198,Value=0.4929),(Time=0.9855,Value=0.5049)),PreInfinityExtrap=RCCE_Constant,PostInfinityExtrap=RCCE_Constant,DefaultValue=340282346638528859811704183484516925440))
    bDrawDebug=True
   )
   > &RigUnit_ChainHarmonics(
    ChainRoot=tendN_1
    DrawWorldOffset=(Rotation=(X=0,Y=0,Z=0,W=1),Translation=(X=0,Y=0,Z=0),Scale3D=(X=1,Y=1,Z=1))
    Pendulum=(bEnabled=False,PendulumStiffness=2,PendulumGravity=(X=0,Y=0,Z=0),PendulumBlend=0.75,PendulumDrag=0.98,PendulumMinimum=0,PendulumMaximum=0.1,PendulumEase=Linear,UnwindAxis=(X=0,Y=1,Z=0),UnwindMinimum=0.2,UnwindMaximum=0.05)
    Reach=(bEnabled=False,ReachTarget=(X=0,Y=0,Z=0),ReachAxis=(X=1,Y=0,Z=0),ReachMinimum=0,ReachMaximum=0,ReachEase=SineEaseInOut)
    Speed=(X=0.25,Y=0.25,Z=0.1)
    Wave=(bEnabled=True,WaveFrequency=(X=0.2,Y=0.2,Z=0.2),WaveAmplitude=(X=0.25,Y=2,Z=1.5),WaveOffset=(X=1,Y=0,Z=2),WaveNoise=(X=8,Y=8,Z=30),WaveMinimum=0,WaveMaximum=0.75,WaveEase=SineEaseIn)
    WaveCurve=(EditorCurveData=(Keys=((Time=-0.0198,Value=0.4929),(Time=0.9855,Value=0.5049)),PreInfinityExtrap=RCCE_Constant,PostInfinityExtrap=RCCE_Constant,DefaultValue=340282346638528859811704183484516925440))
    bDrawDebug=True
   )
  ]}
]
hierarchy [
 b_AR_Root {loc=<0 0 0>, rot=<0 0 0 1>}
 Melee_TwoHanded_Handle {parent=b_AR_Root, loc=<0 0 0>, rot=<0 0 0.7071 -0.7071>}
 tendG {parent=Melee_TwoHanded_Handle, loc=<0.6812 -38.8784 -4.7589>, rot=<-0.3782 -0.1956 0.6355 -0.6441>}
 tendA_1 {parent=Melee_TwoHanded_Handle, loc=<3.7199 -39.7534 2.9136>, rot=<-0.0977 0.2998 0.939 0.1372>}
 tendA_2 {parent=tendA_1, loc=<-0.0773 7.3974 5.6853>, rot=<0.2494 0.0642 0.629 -0.7335>}
 tendA_3 {parent=tendA_2, loc=<4.9704 -0.1769 -0.3665>, rot=<0.2388 -0.9403 0.0058 0.2423>}
 tendA_4 {parent=tendA_3, loc=<2.4977 0 0>, rot=<0 0 0 1>}
 tendF_1 {parent=Melee_TwoHanded_Handle, loc=<-3.6562 -35.6809 -2.537>, rot=<0.7455 -0.4571 0.4737 0.1045>}
 tendF_2a {parent=tendF_1, loc=<5.9757 0.1658 -0.118>, rot=<0.0506 0.6947 -0.0737 -0.7137>}
 tendF_2b {parent=tendF_1, loc=<6.6965 0.2259 -0.1502>, rot=<0.0145 -0.7247 0.0075 -0.6888>}
 tendE {parent=Melee_TwoHanded_Handle, loc=<-0.8005 -33.9354 2.9254>, rot=<-0.0612 -0.4121 0.9072 -0.0582>}
 tendD_1 {parent=Melee_TwoHanded_Handle, loc=<1.725 -48.2169 -17.6179>, rot=<0.6202 0.3417 -0.6128 0.3508>}
 tendD_2 {parent=tendD_1, loc=<7.9022 0.0142 0.1898>, rot=<0.3464 -0.312 0.1531 -0.8713>}
 tendD_3 {parent=tendD_2, loc=<15.6462 -0.0014 0.0936>, rot=<-0.1294 -0.5039 0.7487 -0.4108>}
 tendD_4 {parent=tendD_3, loc=<3.5062 -0.0482 0.0617>, rot=<-0.0233 -0.1061 0.2013 -0.9735>}
 tendD_5 {parent=tendD_4, loc=<3.3769 -0.0684 -0.014>, rot=<-0.0103 -0.0111 -0.0139 -0.9998>}
 tendC_1 {parent=Melee_TwoHanded_Handle, loc=<-2.7421 -47.5494 -11.0102>, rot=<0.1671 -0.6382 -0.2326 -0.7146>}
 tendC_2 {parent=tendC_1, loc=<7.4342 0.0115 -0.0024>, rot=<0.0512 0.37 -0.1254 -0.9191>}
 tendC_3 {parent=tendC_2, loc=<8.3703 -0.1157 -0.1479>, rot=<0.0797 0.1069 0.92 -0.3685>}
 tendC_4 {parent=tendC_3, loc=<2.3036 -0.0823 0.0723>, rot=<-0.0748 0.0317 0.0513 -0.9954>}
 tendB_1 {parent=Melee_TwoHanded_Handle, loc=<2.8677 -50.4817 -9.0056>, rot=<-0.5066 0.5399 0.5609 0.3705>}
 tendB_2 {parent=tendB_1, loc=<15.5795 0.0958 -0.1169>, rot=<-0.1901 -0.8706 -0.0321 -0.4527>}
 tendB_3 {parent=tendB_2, loc=<1.7376 -0.0707 -0.0303>, rot=<0.087 -0.2591 -0.0271 -0.9615>}
 tendB_4 {parent=tendB_3, loc=<2.3138 -0.0043 0.0258>, rot=<0.4954 -0.6466 0.3765 -0.4413>}
 tendI_1 {parent=Melee_TwoHanded_Handle, loc=<-3.1982 -28.764 2.3031>, rot=<0.1384 0.7329 -0.1567 -0.6474>}
 tendI_2 {parent=tendI_1, loc=<4.1537 0 0>, rot=<0.2 0.2188 0.4932 -0.8179>}
 tendI_3 {parent=tendI_2, loc=<2.9524 -0.0132 -0.7653>, rot=<0.5763 -0.5827 0.4195 -0.3904>}
 tendH_1 {parent=Melee_TwoHanded_Handle, loc=<0.2577 -20.4263 -8.9697>, rot=<-0.4729 0.4859 0.358 -0.642>}
 tendH_2 {parent=tendH_1, loc=<9.056 0 0>, rot=<0.4748 -0.3155 0.7689 0.2896>}
 tendH_3 {parent=tendH_2, loc=<4.1003 0 0>, rot=<-0.4264 0.1576 0.344 -0.8216>}
 tendJ_1 {parent=Melee_TwoHanded_Handle, loc=<3.0409 -11.7245 2.1549>, rot=<0.2886 0.654 -0.2984 -0.6324>}
 tendJ_2a {parent=tendJ_1, loc=<4.1871 -0.2662 -0.6658>, rot=<0.0352 -0.2065 -0.1947 -0.9582>}
 tendJ_3a {parent=tendJ_2a, loc=<7.4171 -0.4329 0.8211>, rot=<0.0437 0.1962 -0.2303 -0.9521>}
 tendJ_4a {parent=tendJ_3a, loc=<6.8788 -0.8541 -0.4384>, rot=<-0.0104 0.0961 -0.0926 -0.991>}
 tendJ_5a {parent=tendJ_4a, loc=<10.0011 0 0>, rot=<0.043 -0.1064 0.1617 -0.9802>}
 tendJ_3c {parent=tendJ_2a, loc=<3.322 -2.7342 3.4124>, rot=<-0.1766 0.2442 0.1128 -0.9468>}
 tendJ_2b {parent=tendJ_1, loc=<5.4251 3.2376 -4.1738>, rot=<-0.0083 -0.0146 -0.4946 -0.869>}
 tendJ_3b {parent=tendJ_2b, loc=<6.1766 -0.454 -0.6971>, rot=<0.0484 0.1413 0.0241 -0.9885>}
 tendJ_4b {parent=tendJ_3b, loc=<5.1943 1.0207 -0.6825>, rot=<0 0 0.055 -0.9985>}
 tendK_1 {parent=Melee_TwoHanded_Handle, loc=<-0.0043 -0.1999 8.5629>, rot=<0.4282 0.5627 -0.4282 -0.5627>}
 tendK_2 {parent=tendK_1, loc=<4.1236 0.486 0.0071>, rot=<-0.7778 -0.176 -0.132 -0.5887>}
 tendK_3 {parent=tendK_2, loc=<4.534 0.2097 0.1878>, rot=<0.7575 0.2336 0.1479 -0.5914>}
 tendL_1 {parent=Melee_TwoHanded_Handle, loc=<0.2385 1.0612 -10.4849>, rot=<-0.5414 -0.444 0.5828 0.4124>}
 tendL_2 {parent=tendL_1, loc=<6.1941 -0.186 0.139>, rot=<0.0033 -0.0242 0.067 -0.9975>}
 tendL_3 {parent=tendL_2, loc=<6.7198 -1.0472 -0.0061>, rot=<-0.6123 -0.145 -0.1287 -0.7665>}
 tendM_1 {parent=Melee_TwoHanded_Handle, loc=<2.3594 -0.4277 -7.9423>, rot=<-0.4931 -0.5057 0.524 0.4759>}
 tendM_2 {parent=tendM_1, loc=<9.4022 -0.6295 1.8277>, rot=<0.9609 0.2473 -0.1179 0.041>}
 tendN_1 {parent=Melee_TwoHanded_Handle, loc=<-5.3598 -10.0896 0.3273>, rot=<0.3245 0.6537 -0.3683 -0.5759>}
 tendN_2 {parent=tendN_1, loc=<5.2127 0 0>, rot=<0.0371 -0.2569 -0.258 -0.9306>}
 tendN_3 {parent=tendN_2, loc=<3.2166 0 0.5572>, rot=<0.9582 0.0319 0.0385 0.2815>}
]
